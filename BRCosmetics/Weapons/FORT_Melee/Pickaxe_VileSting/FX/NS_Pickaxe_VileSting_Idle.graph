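# Idle glow effect for the Vile Sting pickaxe, built from one Glow emitter.
# A burst spawns three persistent sprites with per-index offset, size and alignment, tinted by the system TOD value.
# Alpha fades in with system age and differs between gameplay and front end, while a waveform pulses RGB.
# Sprite width varies with camera angle, and the variant index sets a visibility tag with two sprite renderers present.
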NiagaraSystem /BRCosmetics/Weapons/FORT_Melee/Pickaxe_VileSting/FX/NS_Pickaxe_VileSting_Idle
dependencies {
 usedDataInterfaces [&NiagaraDataInterfaceCurve, &NiagaraDataInterfaceCamera, &NiagaraDataInterfaceColorCurve]
 usedDynamicInputs [&ColorFromCurve]
 usedModules [
  &self:SystemSpawnScript.SystemScriptSource.SystemScriptGraph.NiagaraNodeAssignment_0.SetVariables_#1
  &SystemState
  &FN_VariantManager
 ]
}
emitters {Glow {
  emitterUpdateScript [
   &EmitterState(
    Inactive Response=::NewEnumerator1
    Life Cycle Mode=::NewEnumerator1
    Loop Behavior=::NewEnumerator1
    Loop Duration Mode=::NewEnumerator1
    UseLoopDelay=false
   )
   &SpawnBurst_Instantaneous(Spawn Count=3, Use Spawn Probability=false, UseLoopCountLimit=false)
  ]
  particleSpawnScript [
   &InitializeParticle(
    Color=&MultiplyLinearColor(Linear Color=<rgba 1 1 1 0.35>, Scale Factor=&MakeVector4_FromLinearColor(LinearColor=$System.TOD))
    Lifetime=&RandomRangeFloat(Minimum=0.6, Maximum=1)
    Position Offset=&MergeHelper_VectorSwitch(0=<3 0 25>, 1=<5 0 35>, 2=<9 0 43>, 3=?, Index=&ReturnExecIndex(), Num=::NewEnumerator1)
    Sprite Size=&MergeHelper_Vector2Switch(0=<9 50>, 1=<10 40>, 2=<6 30>, 3=?, 4=?, Index=&ReturnExecIndex(), Num=::NewEnumerator1)
    Sprite Size Mode=::NewEnumerator4
    UsePositionOffset=true
    Write Material Random=true
   )
   &SpriteFacingAndAlignment(
    Sprite Alignment=&MergeHelper_VectorSwitch(0=<0.15 0 1>, 1=<0.4 0 1>, 2=<0.7 0 1>, 3=?, Index=&ReturnExecIndex(), Num=::NewEnumerator1)
    WriteAlignment=true
    WriteFacing=true
   )
   &CameraOffset(Camera Offset Amount=-1)
  ]
  particleUpdateScript [
   &ParticleState(Kill Particles When Lifetime Has Elapsed=false)
   /SetVariables(Particles.VisibilityTag=&MakeIntFromFloat(Float=$System.VariantIndex)) {isSetParameters=true}
   &ScaleColor(
    Scale Alpha=&FloatFromCurve(
     FloatCurve=DataInterface() {propertyValues {Curve {keys [<at 0 0 RCIM_Cubic>,<at 1 1 RCIM_Cubic>]}}}
     CurveIndex=$Engine.System.Age
     Scale Curve=&Multiply_Float(
      A=&MergeHelper_FloatSwitch(0=1, 1=1.2, 2=0.8, 3=?, Index=&ReturnExecIndex(), Num=::NewEnumerator1)
      B=&FN_FrontEndFloatModifier(
       Gameplay=1
       Front End=/BRCosmetics/Weapons/FORT_Melee/Pickaxe_GateHound/FX/VariantManager/VariantFloatSelector(
        Original Float:NiagaraFloat=0.2
        Variant 1 Float:NiagaraFloat=0.5
        Variant 2 Float:NiagaraFloat=?
        Variant 3 Float:NiagaraFloat=?
        Variant Amount:NiagaraInt32=0
       )
      )
     )
    )
    Scale RGB=&VectorFromFloat(Value=&Waveform(
      Waveform Count=::NewEnumerator1
      "Waveform [1]"=::NewEnumerator12
      "[1] Frequency"=6
      "[1] Amplitude Min/Max"=<0.5 1>
      "[1] Bias"=0.2
      "[2] Average"=false
      "[2] Input Position"=0
      "[2] Phase"=$Emitter.Age
      "[2] Phase Scale"=1
      "[2] Compound Frequency"=2
      "[2] Frequency"=1.2
      "[2] Amplitude Min/Max"=<0.5 1>
      "[2] Scale"=1
      "[2] Bias"=0
      "[2] Exponent"=1
      "[2] Pulse Duration"=0.1
      "[2] Clip Triangle Min/Max"=<0 0>
      "[2] Random Seed"=0
     ))
    ScaleA=true
    ScaleRGB=true
   )
   /SetVariables(Particles.SpriteSize=&Multiply_Vector2D(
     A=$Particles.Initial.SpriteSize
     B=&MakeVector2D(
      X=&Lerp_Float(
       A=2
       B=1
       Alpha=&Abs_Float(Float:NiagaraFloat=&DotProduct(
         Vector A=&TransformVector(Vector=<1 0 0>, Destination Space=::Simulation)
         Vector B=&ReturnCameraVector(Camera Vectors=::NewEnumerator4, Coordinate Space=::Simulation, Origin Position=$Particles.Position)
        ))
      )
      Y=1
     )
    )) {isSetParameters=true}
  ]
  rendererClasses [&NiagaraSpriteRendererProperties]
  renderers [
   {rendererClass=&NiagaraSpriteRendererProperties}
   {rendererClass=&NiagaraSpriteRendererProperties, rendererIndex=1}
  ]
 }}
summary {
 systemName=NS_Pickaxe_VileSting_Idle
 emitters [{emitterName=Glow, rendererClasses [&NiagaraSpriteRendererProperties]}]
 userVariables [
  User.bIsFrontEnd {type=NiagaraBool}
  User.bIsFrontEndPreview {type=NiagaraBool}
  User.Variant {type=NiagaraFloat, defaultValue=0, description=""}
 ]
}
