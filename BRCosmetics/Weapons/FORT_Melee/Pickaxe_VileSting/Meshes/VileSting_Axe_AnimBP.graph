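# Animation blueprint for the Vile Sting pickaxe mesh, handling its custom idle in the front end.
# The anim graph uses a layered blend per bone to lay the DefaultSlot output over the input pose on the b_AR_Root and Melee_TwoHanded_Handle branches.
# On update, when in the front end, it copies the Frontend_Idle_Bypass curve from the owner's front-end anim instance into the variable driving that blend weight.
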AnimBlueprint /BRCosmetics/Weapons/FORT_Melee/Pickaxe_VileSting/Meshes/VileSting_Axe_AnimBP
target_skeleton=&/VileSting_Axe_Skeleton
graphs [
 AnimGraph {nodes [
   AnimGraphNode_SubInput_0 = Input Pose()
   Output Pose(Result=Layered blend per bone(
     BasePose=@AnimGraphNode_SubInput_0
     BlendPoses_0=Slot 'DefaultSlot'()
     BlendWeights_0=Get Custom Idle Is Playing() -> (Custom Idle Is Playing:double)
    ) {layerSetup [{branchFilters [{boneName=b_AR_Root}, {boneName=Melee_TwoHanded_Handle}]}]}) {layerGroup=None}
  ]}
 EventGraph {nodes [
   Event BlueprintUpdateAnimation()
   > Branch(Condition=Get bIsInFrontEnd())
   > Set Custom Idle Is Playing(Custom Idle Is Playing:double=GetCurveValue(self=GetOwnerFrontEndAnimInstance(), CurveName=Frontend_Idle_Bypass)) -> (Output_Get:double)
   TryGetPawnOwner()
  ]}
]
variables [Custom Idle Is Playing {type=number}]
